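# UI-domain material named Icon_Focus in the InstancedMeshesCluster tech-art tool's Icon folder, apparently intended as a focus icon.
# It holds a single texture sample of the engine's editor target point sprite, S_TargetPoint.
# The graph shows no connection from that sample to a material output, and no parameters or other logic.
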Material /Game/ArtTools/Tech_Art/InstancedMeshesCluster/Icon/Icon_Focus
graph {
 properties {materialDomain=MD_UI, pixelDepthOffsetMode=PDOM_Legacy, usedShadingModels=Default Lit}
 nodes [TextureSample() {texture=/Engine/EditorResources/S_TargetPoint}]
}
